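# Banner message widget for the Movie Render Graph preview window, showing a centered bold text line over a translucent dark border.
# SetMessage takes a Movie Graph Preview Banner Message struct and breaks it into message, text color and background color.
# It applies these to the text block and its containing border, so the caller controls both content and styling.
# The event graph holds only unconnected default events.
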WidgetBlueprint /MovieRenderPipeline/Blueprints/Graph/UI_MovieGraph_PreviewWindowBannerMessage
graphs [
 EventGraph {nodes [
   Event PreConstruct()
   Event Construct()
   Event Tick()
  ]}
 SetMessage {nodes [
   n1 = SetMessage!() -> (then, Message:Movie Graph Preview Banner Message Structure) {is=title K2Node_FunctionEntry}
   n2 = Break Movie Graph Preview Banner Message!(MovieGraphPreviewBannerMessage:"Movie Graph Preview Banner Message Structure (by ref)"=@n1) -> (Message:Text, TextColor:Linear Color Structure, BackgroundColor:Linear Color Structure) {is=title K2Node_BreakStruct}
   > "SetText (Text)"(self=Get MessageTextBlock() -> (MessageTextBlock:Text Object Reference), InText=@n2)
   > SetColorAndOpacity(self=Get MessageTextBlock(), InColorAndOpacity=Make Slate Color(SpecifiedColor=@n2.TextColor))
   > SetBrushColor(self=Get MessageContainer() -> (MessageContainer:Border Object Reference), InBrushColor=@n2.BackgroundColor)
   > Return Node()
  ]}
]
variables []
widgets {
 info {inheritedWidgetCount=0, parentClass=&UserWidget, rootWidgetClass=&Border, widgetCount=2}
 widgets [
  MessageContainer {
   widgetClassPath=&Border
   bIsVariable=true
   widget=MessageContainer
   properties {horizontalAlignment=HAlign_Center, verticalAlignment=VAlign_Center, brushColor=<rgba 0 0 0 0.5>, padding=<ltrb 4 0 4 0>}
  }
  MessageTextBlock {
   widgetClassPath=&TextBlock
   bIsVariable=true
   parent=MessageContainer
   widget=MessageTextBlock
   properties {
    text=MESSAGE
    font {
     bForceMonospaced=false
     bMaterialIsStencil=false
     fontMaterial=null
     fontObject=&Roboto
     letterSpacing=0
     monospacedWidth=1
     size=38
     skewAmount=0
     typefaceFontName=Bold
     outlineSettings {
      bApplyOutlineToDropShadows=false
      bMiteredCorners=false
      bSeparateFillAlpha=false
      outlineMaterial=null
      outlineSize=0
      outlineColor=<rgba 0 0 0 1>
     }
    }
   }
   slot {class=BorderSlot, properties {horizontalAlignment=HAlign_Center, verticalAlignment=VAlign_Center, padding=<ltrb 4 0 4 0>}}
  }
 ]
}
